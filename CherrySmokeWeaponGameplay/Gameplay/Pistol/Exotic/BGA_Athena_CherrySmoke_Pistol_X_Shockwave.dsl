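; Blueprint for the shockwave effect actor of the exotic CherrySmoke pistol in Athena.
; On BeginPlay it plays its shockwave sound cue at the actor's location and activates its Niagara component without resetting it.
; The construction script is empty, and the graph holds only these cosmetic effects, with no damage, knockback or cleanup logic.
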(event EventBeginPlay
  (Audio|PlaySoundatLocation (Variables|Default|GetShockwaveSoundCue) (Transformation|GetActorLocation))
  (Components|Activation|Activate (Variables|Default|GetShockwaveNiagara) false))

(fn ConstructionScript)
